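# Physics asset for the Voyager Remix medium male player body, covering the dynamic left and right shoulder strap clips.
# Each box clip joins a tiny capsule on its parent bone with tight swing and twist limits on two axes, leaving a hinge-like flap.
# Those parent capsules attach to upper-arm twist capsules, which attach to a spine_05 capsule, all with free angular motion.
# The arm and chest capsules approximate the body around the straps.
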PhysicsAsset /BRCosmetics/Characters/Player/Male/Medium/Bodies/M_MED_VoyagerRemix/Meshes/M_MED_Voyager_Remix_Dyn_Physics
bodies {
 shldrstrap_clip_l [{shapeType=Box, extentX=5.51, extentY=6.51, extentZ=1.51, center=<3.2635 -0.0001 0>, rotation=<rot 0 -90 0>}]
 shldrstrap_clip_par_l [{shapeType=Capsule, length=0.505, radius=0.505, rotation=<rot 0 -90 0>}]
 shldrstrap_clip_par_r [{shapeType=Capsule, length=0.505, radius=0.505, rotation=<rot 0 -90 0>}]
 shldrstrap_clip_r [{shapeType=Box, extentX=5.834, extentY=1.3604, extentZ=6.2428, center=<-2.5891 -0.1293 0.0323>, rotation=<rot 0 2.0345 -89.5499>}]
 spine_05 [{shapeType=Capsule, length=7.6901, radius=9.5954, center=<-3.0117 2.0669 0.5887>, rotation=<rot -6.2372 -83.0515 -87.7117>}]
 upperarm_twist_01_l [{shapeType=Capsule, length=3, radius=6.25, center=<11.3416 0.0392 -0.3844>, rotation=<rot 82.5384 0 0>}]
 upperarm_twist_01_r [{shapeType=Capsule, length=2.75, radius=6.25, center=<-11.3416 -0.0392 0.3844>, rotation=<rot 82.5384 0 -180>}]
}
constraints [
 {
  bone1Name=shldrstrap_clip_par_l
  bone2Name=upperarm_twist_01_l
  swing1LimitDegrees=45
  swing1Motion=Free
  swing2LimitDegrees=45
  swing2Motion=Free
  twistLimitDegrees=45
  twistMotion=Free
 }
 {bone1Name=shldrstrap_clip_l, bone2Name=shldrstrap_clip_par_l, swing1LimitDegrees=1, swing2LimitDegrees=45, twistLimitDegrees=1}
 {
  bone1Name=upperarm_twist_01_l
  bone2Name=spine_05
  swing1LimitDegrees=45
  swing1Motion=Free
  swing2LimitDegrees=45
  swing2Motion=Free
  twistLimitDegrees=45
  twistMotion=Free
 }
 {
  bone1Name=shldrstrap_clip_par_r
  bone2Name=upperarm_twist_01_r
  swing1LimitDegrees=45
  swing1Motion=Free
  swing2LimitDegrees=45
  swing2Motion=Free
  twistLimitDegrees=45
  twistMotion=Free
 }
 {bone1Name=shldrstrap_clip_r, bone2Name=shldrstrap_clip_par_r, swing1LimitDegrees=1, swing2LimitDegrees=45, twistLimitDegrees=1}
 {
  bone1Name=upperarm_twist_01_r
  bone2Name=spine_05
  swing1LimitDegrees=45
  swing1Motion=Free
  swing2LimitDegrees=45
  swing2Motion=Free
  twistLimitDegrees=45
  twistMotion=Free
 }
]
