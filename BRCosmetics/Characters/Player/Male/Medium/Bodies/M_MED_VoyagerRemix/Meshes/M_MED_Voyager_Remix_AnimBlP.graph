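# Post-process style AnimBlueprint for the Voyager Remix medium male body, adding secondary motion to straps and shoe tongues over the copied parent-mesh pose.
# Weighted multi-bone constraints place strap bases, then leg-driven bone offsets, spring AnimDynamics and a RigidBody node add movement.
# EventGraph measures arm and thigh rotations plus rib-strap-to-arm distances; leg and distance values scale node alphas.
# Shoulder-strap nodes and rib-strap offset transforms are disconnected from Output Pose.
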AnimBlueprint /BRCosmetics/Characters/Player/Male/Medium/Bodies/M_MED_VoyagerRemix/Meshes/M_MED_Voyager_Remix_AnimBlP
target_skeleton=&/M_MED_Voyager_Remix_Skeleton
graphs [
 AnimGraph {nodes [
   Anim Dynamics - Bone: shldrstrap_clip_l(Alpha=1, ExternalForce="0,0,0") {
    angularDampingOverride=1
    angularSpringConstant=1000
    bAngularSpring=true
    bGravityOverrideInSimSpace=true
    bOverrideAngularDamping=true
    gravityScale=0.35
    simulationSpace=RootRelative
    alphaBoolBlend {blendOption=Linear}
    boundBone {boneName=shldrstrap_clip_l}
    componentAppliedLinearAccClamp=<3000 3000 3000>
    componentLinearAccScale=<0.2 0.2 0.2>
    componentLinearVelScale=<0.2 0.2 0.2>
    physicsBodyDefinitions [{
      sphereCollisionRadius=0
      boundBone {boneName=shldrstrap_clip_l}
      boxExtents=<5 5 5>
      constraintSetup {angularTargetAxis=AxisY, twistAxis=AxisZ, angularLimitsMax=<0 30 0>, angularLimitsMin=<0 -30 0>, angularTarget=<0 0 0>}
      localJointOffset=<0 0 -5>
     }]
   }
   n1 = Look At - Bone: shldrstrap_par_l(Alpha=0.7) {
    bUseLookUpAxis=true
    alphaBoolBlend {blendOption=Linear}
    boneToModify {boneName=shldrstrap_par_l}
    lookAtLocation {x=0}
    lookAtTarget {boneReference {boneName=lowerarm_l}}
    lookAt_Axis {axis=<1 0>}
    lookUp_Axis {axis {x=0, z=1}}
   }
   n2 = "Transform (Modify) Bone - Bone: shldrstrap_par_l"(ComponentPose=@n1, Alpha=Get armUp_L() -> (armUp_L:double), Translation="0,0,1", Rotation="45,0,0", Scale="1,1,1") {
    rotationMode=BMM_Additive
    rotationSpace=BCS_BoneSpace
    translationMode=BMM_Additive
    translationSpace=BCS_BoneSpace
    alphaBoolBlend {blendOption=Linear}
    alphaScaleBiasClamp {bInterpResult=true, bMapRange=true, inRange {max=35}}
    boneToModify {boneName=shldrstrap_par_l}
    rotation {pitch=45}
    translation {z=1}
   }
   n3 = Look At - Bone: shldrstrap_par_r(ComponentPose=@n2, Alpha=0.7) {
    bUseLookUpAxis=true
    alphaBoolBlend {blendOption=Linear}
    boneToModify {boneName=shldrstrap_par_r}
    lookAtLocation {x=0}
    lookAtTarget {boneReference {boneName=lowerarm_r}}
    lookAt_Axis {axis=<-1 0>}
    lookUp_Axis {axis {x=0, z=1}}
   }
   "Transform (Modify) Bone - Bone: shldrstrap_par_r"(ComponentPose=@n3, Alpha=Get armUp_R() -> (armUp_R:double), Translation="0,0,-2", Rotation="70,0,0", Scale="1,1,1") {
    rotationMode=BMM_Additive
    rotationSpace=BCS_BoneSpace
    translationMode=BMM_Additive
    translationSpace=BCS_BoneSpace
    alphaBoolBlend {blendOption=Linear}
    alphaScaleBiasClamp {bInterpResult=true, bMapRange=true, inRange {max=20}}
    boneToModify {boneName=shldrstrap_par_r}
    rotation {pitch=70}
    translation {z=-2}
   }
   n4 = "Transform (Modify) Bone - Bone: ribstrap_base_l"(Alpha=Get armToStrap_L() -> (armToStrap_L:double), Translation="0,0,0", Rotation="-6,0,0", Scale="1,1,1") {
    rotationMode=BMM_Additive
    rotationSpace=BCS_BoneSpace
    translationMode=BMM_Additive
    translationSpace=BCS_BoneSpace
    alphaBoolBlend {blendOption=Linear}
    alphaScaleBiasClamp {bMapRange=true, outRange {max=0, min=1}}
    boneToModify {boneName=ribstrap_base_l}
    rotation {pitch=-6}
   }
   n5 = "Transform (Modify) Bone - Bone: ribstrap_base_r"(ComponentPose=@n4, Alpha=Get armToStrap_R() -> (armToStrap_R:double), Translation="0,0,0", Rotation="-15,0,0", Scale="1,1,1") {
    rotationMode=BMM_Additive
    rotationSpace=BCS_BoneSpace
    translationMode=BMM_Additive
    translationSpace=BCS_BoneSpace
    alphaBoolBlend {blendOption=Linear}
    alphaScaleBiasClamp {bMapRange=true, outRange {max=0, min=1}}
    boneToModify {boneName=ribstrap_base_r}
    rotation {pitch=-15}
   }
   n6 = "Transform (Modify) Bone - Bone: ribstrap_base_l"(ComponentPose=@n5, Alpha=1, Translation="0,0,0", Rotation="10,10,0", Scale="1,1,1") {
    rotationMode=BMM_Additive
    rotationSpace=BCS_BoneSpace
    translationMode=BMM_Additive
    translationSpace=BCS_BoneSpace
    alphaBoolBlend {blendOption=Linear}
    boneToModify {boneName=ribstrap_base_l}
    rotation {pitch=10, yaw=10}
   }
   "Transform (Modify) Bone - Bone: ribstrap_base_r"(ComponentPose=@n6, Alpha=1, Translation="0,0,0", Rotation="15,15,0", Scale="1,1,1") {
    rotationMode=BMM_Additive
    rotationSpace=BCS_BoneSpace
    translationMode=BMM_Additive
    translationSpace=BCS_BoneSpace
    alphaBoolBlend {blendOption=Linear}
    boneToModify {boneName=ribstrap_base_r}
    rotation {pitch=15, yaw=15}
   }
   n7 = "Constraint - ribstrap_base_l (PPPP)"!(
    ComponentPose=Local To Component(LocalPose=Copy Pose From Mesh() {bUseAttachedParent=true})
    Alpha=1
    ConstraintWeights_0=0.4613
    ConstraintWeights_1:float=0.2483
    ConstraintWeights_2:float=0.0616
    ConstraintWeights_3:float=0.2287
   ) -> (Pose) {
    alphaBoolBlend {blendOption=Linear}
    boneToModify {boneName=ribstrap_base_l}
    constraintSetup [
     {targetBone {boneName=clavicle_l}}
     {targetBone {boneName=spine_04}}
     {targetBone {boneName=spine_05}}
     {targetBone {boneName=upperarm_twist_01_l}}
    ]
    constraintWeights [1,1,1,1]
   }
   n8 = "Constraint - ribstrap_base_r (PPPP)"!(
    ComponentPose=@n7
    Alpha=1
    ConstraintWeights_0=0.2287
    ConstraintWeights_1:float=0.4613
    ConstraintWeights_2:float=0.2483
    ConstraintWeights_3:float=0.0616
   ) -> (Pose) {
    alphaBoolBlend {blendOption=Linear}
    boneToModify {boneName=ribstrap_base_r}
    constraintSetup [
     {targetBone {boneName=upperarm_twist_01_r}}
     {targetBone {boneName=clavicle_r}}
     {targetBone {boneName=spine_04}}
     {targetBone {boneName=spine_05}}
    ]
    constraintWeights [1,1,1,1]
   }
   n9 = "Constraint - shldrstrap_clip_base_l (PPPPP)"!(
    ComponentPose=@n8
    Alpha=1
    ConstraintWeights_0=0.233
    ConstraintWeights_1:float=0.206
    ConstraintWeights_2:float=0.016
    ConstraintWeights_3:float=0.103
    ConstraintWeights_4:float=0.442
   ) -> (Pose) {
    alphaBoolBlend {blendOption=Linear}
    boneToModify {boneName=shldrstrap_clip_base_l}
    constraintSetup [
     {targetBone {boneName=clavicle_l}}
     {targetBone {boneName=upperarm_twist_02_l}}
     {targetBone {boneName=spine_05}}
     {targetBone {boneName=upperarm_l}}
     {targetBone {boneName=upperarm_twist_01_l}}
    ]
    constraintWeights [1,1,1,1,1]
   }
   n10 = "Constraint - shldrstrap_clip_base_r (PPPPP)"!(
    ComponentPose=@n9
    Alpha=1
    ConstraintWeights_0=0.233
    ConstraintWeights_1:float=0.206
    ConstraintWeights_2:float=0.016
    ConstraintWeights_3:float=0.103
    ConstraintWeights_4:float=0.442
   ) -> (Pose) {
    alphaBoolBlend {blendOption=Linear}
    boneToModify {boneName=shldrstrap_clip_base_r}
    constraintSetup [
     {targetBone {boneName=clavicle_r}}
     {targetBone {boneName=upperarm_twist_02_r}}
     {targetBone {boneName=spine_05}}
     {targetBone {boneName=upperarm_r}}
     {targetBone {boneName=upperarm_twist_01_r}}
    ]
    constraintWeights [1,1,1,1,1]
   }
   n11 = "Constraint - trapstrap_base_l (PPPPPPP)"!(
    ComponentPose=@n10
    Alpha=1
    ConstraintWeights_0=0.5862
    ConstraintWeights_1:float=0.0299
    ConstraintWeights_2:float=0.0268
    ConstraintWeights_3:float=0.0118
    ConstraintWeights_4:float=0.2192
    ConstraintWeights_5:float=0.1156
    ConstraintWeights_6:float=0.0105
   ) -> (Pose) {
    alphaBoolBlend {blendOption=Linear}
    boneToModify {boneName=trapstrap_base_l}
    constraintSetup [
     {targetBone {boneName=clavicle_l}}
     {targetBone {boneName=upperarm_twist_02_l}}
     {targetBone {boneName=spine_03}}
     {targetBone {boneName=spine_04}}
     {targetBone {boneName=spine_05}}
     {targetBone {boneName=upperarm_twist_01_l}}
     {targetBone {boneName=upperarm_l}}
    ]
    constraintWeights [1,1,1,1,1,1,1]
   }
   n12 = "Constraint - trapstrap_base_r (PPPPPPP)"!(
    ComponentPose=@n11
    Alpha=1
    ConstraintWeights_0=0.0105
    ConstraintWeights_1:float=0.1156
    ConstraintWeights_2:float=0.0299
    ConstraintWeights_3:float=0.0268
    ConstraintWeights_4:float=0.0118
    ConstraintWeights_5:float=0.2192
    ConstraintWeights_6:float=0.5862
   ) -> (Pose) {
    alphaBoolBlend {blendOption=Linear}
    boneToModify {boneName=trapstrap_base_r}
    constraintSetup [
     {targetBone {boneName=upperarm_r}}
     {targetBone {boneName=upperarm_twist_01_r}}
     {targetBone {boneName=upperarm_twist_02_r}}
     {targetBone {boneName=spine_03}}
     {targetBone {boneName=spine_04}}
     {targetBone {boneName=spine_05}}
     {targetBone {boneName=clavicle_r}}
    ]
    constraintWeights [1,1,1,1,1,1,1]
   }
   n13 = "Constraint - beltstrap_fr_base_l (PPP)"!(ComponentPose=@n12, Alpha=1, ConstraintWeights_0=0.0671, ConstraintWeights_1:float=0.1757, ConstraintWeights_2:float=0.7571) -> (Pose) {
    alphaBoolBlend {blendOption=Linear}
    boneToModify {boneName=beltstrap_fr_base_l}
    constraintSetup [{targetBone {boneName=spine_01}}, {targetBone {boneName=thigh_twist_01_l}}, {targetBone {boneName=pelvis}}]
    constraintWeights [1,1,1]
   }
   n14 = "Constraint - beltstrap_bk_base_l (PPP)"!(ComponentPose=@n13, Alpha=1, ConstraintWeights_0=0.0998, ConstraintWeights_1:float=0.0289, ConstraintWeights_2:float=0.8713) -> (Pose) {
    alphaBoolBlend {blendOption=Linear}
    boneToModify {boneName=beltstrap_bk_base_l}
    constraintSetup [{targetBone {boneName=spine_01}}, {targetBone {boneName=thigh_twist_01_l}}, {targetBone {boneName=pelvis}}]
    constraintWeights [1,1,1]
   }
   n15 = "Constraint - beltstrap_fr_base_r (PPP)"!(ComponentPose=@n14, Alpha=1, ConstraintWeights_0=0.064, ConstraintWeights_1:float=0.164, ConstraintWeights_2:float=0.772) -> (Pose) {
    alphaBoolBlend {blendOption=Linear}
    boneToModify {boneName=beltstrap_fr_base_r}
    constraintSetup [{targetBone {boneName=spine_01}}, {targetBone {boneName=thigh_twist_01_r}}, {targetBone {boneName=pelvis}}]
    constraintWeights [1,1,1]
   }
   n16 = "Constraint - beltstrap_bk_base_r (PPP)"!(ComponentPose=@n15, Alpha=1, ConstraintWeights_0=0.0919, ConstraintWeights_1:float=0.0376, ConstraintWeights_2:float=0.8705) -> (Pose) {
    alphaBoolBlend {blendOption=Linear}
    boneToModify {boneName=beltstrap_bk_base_r}
    constraintSetup [{targetBone {boneName=spine_01}}, {targetBone {boneName=thigh_twist_01_r}}, {targetBone {boneName=pelvis}}]
    constraintWeights [1,1,1]
   }
   n17 = "Transform (Modify) Bone - Bone: beltstrap_fr_par_l"(ComponentPose=@n16, Alpha=Get legUp_L() -> (legUp_L:double), Translation="0,0,0", Rotation="120,0,0", Scale="1,1,1") {
    rotationMode=BMM_Additive
    rotationSpace=BCS_BoneSpace
    translationMode=BMM_Additive
    translationSpace=BCS_BoneSpace
    alphaBoolBlend {blendOption=Linear}
    alphaScaleBiasClamp {bMapRange=true, inRange {max=-20, min=-110}, outRange {max=0, min=1}}
    boneToModify {boneName=beltstrap_fr_par_l}
    rotation {pitch=120}
   }
   n18 = "Transform (Modify) Bone - Bone: beltstrap_fr_par_r"(ComponentPose=@n17, Alpha=Get legUp_R() -> (legUp_R:double), Translation="0,0,0", Rotation="120,0,0", Scale="1,1,1") {
    rotationMode=BMM_Additive
    rotationSpace=BCS_BoneSpace
    translationMode=BMM_Additive
    translationSpace=BCS_BoneSpace
    alphaBoolBlend {blendOption=Linear}
    alphaScaleBiasClamp {bMapRange=true, inRange {max=-20, min=-110}, outRange {max=0, min=1}}
    boneToModify {boneName=beltstrap_fr_par_r}
    rotation {pitch=120}
   }
   n19 = Anim Dynamics - Bone: shoeTongue_fr_l(ComponentPose=@n18, Alpha=1, ExternalForce="0,0,0") {
    angularDampingOverride=1
    angularSpringConstant=1000
    bAngularSpring=true
    bGravityOverrideInSimSpace=true
    bOverrideAngularDamping=true
    gravityScale=0.35
    simulationSpace=RootRelative
    alphaBoolBlend {blendOption=Linear}
    boundBone {boneName=shoeTongue_fr_l}
    componentAppliedLinearAccClamp=<3000 3000 3000>
    componentLinearAccScale=<0.2 0.2 0.2>
    componentLinearVelScale=<0.2 0.2 0.2>
    physicsBodyDefinitions [{
      sphereCollisionRadius=0
      boundBone {boneName=shoeTongue_fr_l}
      boxExtents=<5 5 5>
      constraintSetup {angularTargetAxis=AxisZ, angularLimitsMax=<0 2 3>, angularLimitsMin=<0 -2 -2>, angularTarget=<0 0 0>}
      localJointOffset=<1 0 0>
     }]
   }
   n20 = Anim Dynamics - Bone: shoeTongue_fr_r(ComponentPose=@n19, Alpha=1, ExternalForce="0,0,0") {
    angularDampingOverride=1
    angularSpringConstant=1000
    bAngularSpring=true
    bGravityOverrideInSimSpace=true
    bOverrideAngularDamping=true
    gravityScale=0.35
    simulationSpace=RootRelative
    alphaBoolBlend {blendOption=Linear}
    boundBone {boneName=shoeTongue_fr_r}
    componentAppliedLinearAccClamp=<3000 3000 3000>
    componentLinearAccScale=<0.2 0.2 0.2>
    componentLinearVelScale=<0.2 0.2 0.2>
    physicsBodyDefinitions [{
      sphereCollisionRadius=0
      boundBone {boneName=shoeTongue_fr_r}
      boxExtents=<5 5 5>
      constraintSetup {angularTargetAxis=AxisZ, angularLimitsMax=<0 2 3>, angularLimitsMin=<0 -2 -2>, angularTarget=<0 0 0>}
      localJointOffset=<-1 0 0>
     }]
   }
   n21 = Anim Dynamics - Bone: shoeTongue_bk_l(ComponentPose=@n20, Alpha=1, ExternalForce="0,0,0") {
    angularDampingOverride=1
    angularSpringConstant=1000
    bAngularSpring=true
    bGravityOverrideInSimSpace=true
    bOverrideAngularDamping=true
    gravityScale=0.35
    simulationSpace=RootRelative
    alphaBoolBlend {blendOption=Linear}
    boundBone {boneName=shoeTongue_bk_l}
    componentAppliedLinearAccClamp=<3000 3000 3000>
    componentLinearAccScale=<0.2 0.2 0.2>
    componentLinearVelScale=<0.2 0.2 0.2>
    physicsBodyDefinitions [{
      sphereCollisionRadius=0
      boundBone {boneName=shoeTongue_bk_l}
      boxExtents=<5 10 10>
      constraintSetup {angularTargetAxis=AxisZ, angularLimitsMax=<0 5 20>, angularLimitsMin=<0 -5 -20>, angularTarget=<0 0 0>}
      localJointOffset=<3 0 0>
     }]
   }
   n22 = Anim Dynamics - Bone: shoeTongue_bk_r(ComponentPose=@n21, Alpha=1, ExternalForce="0,0,0") {
    angularDampingOverride=1
    angularSpringConstant=1000
    bAngularSpring=true
    bGravityOverrideInSimSpace=true
    bOverrideAngularDamping=true
    gravityScale=0.35
    simulationSpace=RootRelative
    alphaBoolBlend {blendOption=Linear}
    boundBone {boneName=shoeTongue_bk_r}
    componentAppliedLinearAccClamp=<3000 3000 3000>
    componentLinearAccScale=<0.2 0.2 0.2>
    componentLinearVelScale=<0.2 0.2 0.2>
    physicsBodyDefinitions [{
      sphereCollisionRadius=0
      boundBone {boneName=shoeTongue_bk_r}
      boxExtents=<5 10 10>
      constraintSetup {angularTargetAxis=AxisZ, angularLimitsMax=<0 5 20>, angularLimitsMin=<0 -5 -20>, angularTarget=<0 0 0>}
      localJointOffset=<-3 0 0>
     }]
   }
   n23 = Anim Dynamics - Bone: beltstrap_bk_l(ComponentPose=@n22, Alpha=1, ExternalForce="0,0,0") {
    angularDampingOverride=1
    angularSpringConstant=500
    bAngularSpring=true
    bGravityOverrideInSimSpace=true
    bOverrideAngularDamping=true
    gravityScale=0.35
    simulationSpace=RootRelative
    alphaBoolBlend {blendOption=Linear}
    boundBone {boneName=beltstrap_bk_l}
    componentAppliedLinearAccClamp=<2000 2000 2000>
    componentLinearAccScale=<0.2 0.2 0.4>
    componentLinearVelScale=<0.2 0.2 0.4>
    physicsBodyDefinitions [{
      sphereCollisionRadius=0
      boundBone {boneName=beltstrap_bk_l}
      boxExtents=<10 20 20>
      constraintSetup {angularLimitsMax=<0 5 20>, angularLimitsMin=<0 -5 -30>, angularTarget=<0 0 0>}
      localJointOffset=<10 0 0>
     }]
   }
   n24 = Anim Dynamics - Bone: beltstrap_bk_r(ComponentPose=@n23, Alpha=1, ExternalForce="0,0,0") {
    angularDampingOverride=1
    angularSpringConstant=500
    bAngularSpring=true
    bGravityOverrideInSimSpace=true
    bOverrideAngularDamping=true
    gravityScale=0.35
    simulationSpace=RootRelative
    alphaBoolBlend {blendOption=Linear}
    boundBone {boneName=beltstrap_bk_r}
    componentAppliedLinearAccClamp=<3000 3000 3000>
    componentLinearAccScale=<0.2 0.2 0.4>
    componentLinearVelScale=<0.2 0.2 0.4>
    physicsBodyDefinitions [{
      sphereCollisionRadius=0
      boundBone {boneName=beltstrap_bk_r}
      boxExtents=<10 20 20>
      constraintSetup {angularLimitsMax=<0 5 20>, angularLimitsMin=<0 -5 -30>, angularTarget=<0 0 0>}
      localJointOffset=<-10 0 0>
     }]
   }
   n25 = Get legUp_L()
   n26 = Anim Dynamics - Bone: beltstrap_fr_l(ComponentPose=@n24, Alpha=@n25, ExternalForce="0,0,0") {
    angularDampingOverride=1
    angularSpringConstant=500
    bAngularSpring=true
    bGravityOverrideInSimSpace=true
    bOverrideAngularDamping=true
    gravityScale=0.35
    simulationSpace=RootRelative
    alphaBoolBlend {blendOption=Linear}
    alphaScaleBiasClamp {bMapRange=true, inRange {max=-70, min=-110}, outRange {min=0.2}}
    boundBone {boneName=beltstrap_fr_l}
    componentAppliedLinearAccClamp=<2000 2000 2000>
    componentLinearAccScale=<0.2 0.2 0.2>
    componentLinearVelScale=<0.2 0.2 0.2>
    physicsBodyDefinitions [{
      sphereCollisionRadius=0
      boundBone {boneName=beltstrap_fr_l}
      boxExtents=<10 20 20>
      constraintSetup {angularLimitsMax=<0 5 10>, angularLimitsMin=<0 -30 -10>, angularTarget=<0 0 0>}
      localJointOffset=<10 0 0>
     }]
   }
   n27 = Anim Dynamics - Bone: beltstrap_fr_clip_l(ComponentPose=@n26, Alpha=@n25, ExternalForce="0,0,0") {
    angularDampingOverride=1
    angularSpringConstant=500
    bAngularSpring=true
    bGravityOverrideInSimSpace=true
    bOverrideAngularDamping=true
    gravityScale=0.35
    simulationSpace=RootRelative
    alphaBoolBlend {blendOption=Linear}
    alphaScaleBiasClamp {bMapRange=true, inRange {max=-70, min=-110}, outRange {min=0.2}}
    boundBone {boneName=beltstrap_fr_clip_l}
    componentAppliedLinearAccClamp=<2000 2000 2000>
    componentLinearAccScale=<0.2 0.2 0.2>
    componentLinearVelScale=<0.2 0.2 0.2>
    physicsBodyDefinitions [{
      sphereCollisionRadius=0
      boundBone {boneName=beltstrap_fr_clip_l}
      boxExtents=<10 20 20>
      constraintSetup {angularLimitsMax=<0 0 0>, angularLimitsMin=<0 -15 0>, angularTarget=<0 0 0>}
      localJointOffset=<10 0 0>
     }]
   }
   n28 = Get legUp_R()
   n29 = Anim Dynamics - Bone: beltstrap_fr_r(ComponentPose=@n27, Alpha=@n28, ExternalForce="0,0,0") {
    angularDampingOverride=1
    angularSpringConstant=500
    bAngularSpring=true
    bGravityOverrideInSimSpace=true
    bOverrideAngularDamping=true
    gravityScale=0.35
    simulationSpace=RootRelative
    alphaBoolBlend {blendOption=Linear}
    alphaScaleBiasClamp {bMapRange=true, inRange {max=-70, min=-110}, outRange {min=0.2}}
    boundBone {boneName=beltstrap_fr_r}
    componentAppliedLinearAccClamp=<2000 2000 2000>
    componentLinearAccScale=<0.2 0.2 0.2>
    componentLinearVelScale=<0.2 0.2 0.2>
    physicsBodyDefinitions [{
      sphereCollisionRadius=0
      boundBone {boneName=beltstrap_fr_r}
      boxExtents=<10 20 20>
      constraintSetup {angularLimitsMax=<0 5 10>, angularLimitsMin=<0 -30 -10>, angularTarget=<0 0 0>}
      localJointOffset=<-10 0 0>
     }]
   }
   n30 = Anim Dynamics - Bone: beltstrap_fr_clip_r(ComponentPose=@n29, Alpha=@n28, ExternalForce="0,0,0") {
    angularDampingOverride=1
    angularSpringConstant=500
    bAngularSpring=true
    bGravityOverrideInSimSpace=true
    bOverrideAngularDamping=true
    gravityScale=0.35
    simulationSpace=RootRelative
    alphaBoolBlend {blendOption=Linear}
    alphaScaleBiasClamp {bMapRange=true, inRange {max=-70, min=-110}, outRange {min=0.2}}
    boundBone {boneName=beltstrap_fr_clip_r}
    componentAppliedLinearAccClamp=<2000 2000 2000>
    componentLinearAccScale=<0.2 0.2 0.2>
    componentLinearVelScale=<0.2 0.2 0.2>
    physicsBodyDefinitions [{
      sphereCollisionRadius=0
      boundBone {boneName=beltstrap_fr_clip_r}
      boxExtents=<10 20 20>
      constraintSetup {angularLimitsMax=<0 0 0>, angularLimitsMin=<0 -15 0>, angularTarget=<0 0 0>}
      localJointOffset=<-10 0 0>
     }]
   }
   n31 = Anim Dynamics - Bone: trapstrap_l(ComponentPose=@n30, Alpha=1, ExternalForce="0,0,0") {
    angularDampingOverride=0.9
    angularSpringConstant=1000
    bAngularSpring=true
    bGravityOverrideInSimSpace=true
    bOverrideAngularDamping=true
    gravityScale=0.35
    simulationSpace=RootRelative
    alphaBoolBlend {blendOption=Linear}
    boundBone {boneName=trapstrap_l}
    componentAppliedLinearAccClamp=<1000 1000 1000>
    componentLinearAccScale=<0.2 0.2 0.2>
    componentLinearVelScale=<0.2 0.2 0.2>
    physicsBodyDefinitions [{
      sphereCollisionRadius=0
      boundBone {boneName=trapstrap_l}
      boxExtents=<10 10 5>
      constraintSetup {angularTargetAxis=AxisY, angularLimitsMax=<0 5 0>, angularLimitsMin=<0 -5 0>, angularTarget=<0 0 0>}
      localJointOffset=<0 0 -5>
     }]
   }
   n32 = Anim Dynamics - Bone: trapstrap_r(ComponentPose=@n31, Alpha=1, ExternalForce="0,0,0") {
    angularDampingOverride=0.9
    angularSpringConstant=1000
    bAngularSpring=true
    bGravityOverrideInSimSpace=true
    bOverrideAngularDamping=true
    gravityScale=0.35
    simulationSpace=RootRelative
    alphaBoolBlend {blendOption=Linear}
    boundBone {boneName=trapstrap_r}
    componentAppliedLinearAccClamp=<1000 1000 1000>
    componentLinearAccScale=<0.2 0.2 0.2>
    componentLinearVelScale=<0.2 0.2 0.2>
    physicsBodyDefinitions [{
      sphereCollisionRadius=0
      boundBone {boneName=trapstrap_r}
      boxExtents=<10 10 5>
      constraintSetup {angularTargetAxis=AxisY, angularLimitsMax=<0 5 0>, angularLimitsMin=<0 -5 0>, angularTarget=<0 0 0>}
      localJointOffset=<0 0 5>
     }]
   }
   n33 = Anim Dynamics - Bone: ribstrap_l(ComponentPose=@n32, Alpha=Get armToStrap_L(), ExternalForce="0,0,0") {
    angularDampingOverride=1
    angularSpringConstant=1000
    bAngularSpring=true
    bGravityOverrideInSimSpace=true
    bOverrideAngularDamping=true
    gravityScale=0.35
    simulationSpace=RootRelative
    alphaBoolBlend {blendOption=Linear}
    alphaScaleBiasClamp {bMapRange=true, outRange {min=0.3}}
    boundBone {boneName=ribstrap_l}
    componentAppliedLinearAccClamp=<3000 3000 3000>
    componentLinearAccScale=<0.2 0.2 0.2>
    componentLinearVelScale=<0.2 0.2 0.2>
    physicsBodyDefinitions [{
      sphereCollisionRadius=0
      boundBone {boneName=ribstrap_l}
      boxExtents=<5 20 20>
      constraintSetup {angularTargetAxis=AxisY, angularLimitsMax=<5 20 20>, angularLimitsMin=<-5 -10 -15>, angularTarget=<0 0 0>}
      localJointOffset=<10 0 0>
     }]
   }
   n34 = Anim Dynamics - Bone: ribstrap_r(ComponentPose=@n33, Alpha=Get armToStrap_R(), ExternalForce="0,0,0") {
    angularDampingOverride=1
    angularSpringConstant=1000
    bAngularSpring=true
    bGravityOverrideInSimSpace=true
    bOverrideAngularDamping=true
    gravityScale=0.35
    simulationSpace=RootRelative
    alphaBoolBlend {blendOption=Linear}
    alphaScaleBiasClamp {bMapRange=true, outRange {min=0.4}}
    boundBone {boneName=ribstrap_r}
    componentAppliedLinearAccClamp=<3000 3000 3000>
    componentLinearAccScale=<0.2 0.2 0.2>
    componentLinearVelScale=<0.2 0.2 0.2>
    physicsBodyDefinitions [{
      sphereCollisionRadius=0
      boundBone {boneName=ribstrap_r}
      boxExtents=<5 20 20>
      constraintSetup {angularTargetAxis=AxisY, angularLimitsMax=<5 20 20>, angularLimitsMin=<-5 -10 -10>, angularTarget=<0 0 0>}
      localJointOffset=<-10 0 0>
     }]
   }
   n35 = RigidBody(ComponentPose=@n34) {
    bClampLinearTranslationLimitToRefPose=true
    bOverrideWorldGravity=true
    overridePhysicsAsset=&/M_MED_Voyager_Remix_Dyn_Physics
    simulationSpace=BaseBoneSpace
    alphaBoolBlend {blendOption=Linear}
    baseBoneRef {boneName=spine_01}
    componentAppliedLinearAccClamp=<3000 3000 3000>
    componentLinearAccScale=<0.3 0.3 0.3>
    componentLinearVelScale=<0.3 0.3 0.3>
    overrideWorldGravity {z=-980}
   }
   Output Pose(Result=Component To Local(ComponentPose=@n35)) {layerGroup=None}
  ]}
 EventGraph {nodes [
   Event BlueprintUpdateAnimation()
   > n1 = Sequence!(execute) -> (then_0, then_1, then_2:Exec, then_3:Exec, then_4:Exec, then_5:Exec)
   n2 = GetOwnerSkeletalMeshComponent()
   n3 = GetDeltaTransformFromRefPose(self=@n2, BoneName=upperarm_l, BaseName=clavicle_l)
   > Set armUp_L(armUp_L:double=BreakRotator(InRot=BreakTransform(InTransform=@n3).Rotation).Pitch) -> (Output_Get:double)
   n4 = GetDeltaTransformFromRefPose(self=@n2, BoneName=upperarm_r, BaseName=clavicle_r)
   Set armUp_R(execute=@n1.then_1, armUp_R:double=BreakRotator(InRot=BreakTransform(InTransform=@n4).Rotation).Pitch) -> (Output_Get:double)
   n5 = GetDeltaTransformFromRefPose(self=@n2, BoneName=thigh_l, BaseName=pelvis)
   Set legUp_L(execute=@n1.then_2, legUp_L:double=BreakRotator(InRot=BreakTransform(InTransform=@n5).Rotation).Yaw) -> (Output_Get:double)
   n6 = GetDeltaTransformFromRefPose(self=@n2, BoneName=thigh_r, BaseName=pelvis)
   Set legUp_R(execute=@n1.then_3, legUp_R:double=BreakRotator(InRot=BreakTransform(InTransform=@n6).Rotation).Yaw) -> (Output_Get:double)
   n7 = Set armToStrap_R(
    execute=@n1.then_4
    armToStrap_R:double=Get Distance Between Two Sockets!(
     Component:Skeletal Mesh Component Object Reference=@n2
     SocketOrBoneNameA:Name=ribstrap_r
     SocketSpaceA:ERelativeTransformSpace Enum=RTS_World
     SocketOrBoneNameB:Name=upperarm_twist_02_r
     SocketSpaceB:ERelativeTransformSpace Enum=RTS_World
     bRemapRange:bool=true
     InRangeMin:float=120
     InRangeMax:float=130
     OutRangeMin:float=0
     OutRangeMax:float=1
    ) -> (ReturnValue:float) {is=title K2Node_CallFunction}
   ) -> (Output_Get:double)
   Set armToStrap_L(
    execute=@n1.then_5
    armToStrap_L:double=Get Distance Between Two Sockets(
     Component=@n2
     SocketOrBoneNameA=ribstrap_l
     SocketSpaceA=RTS_World
     SocketOrBoneNameB=upperarm_twist_02_l
     SocketSpaceB=RTS_World
     bRemapRange=true
     InRangeMin=110
     InRangeMax=120
     OutRangeMin=0
     OutRangeMax=1
    )
   ) -> (Output_Get:double)
   TryGetPawnOwner()
   "To String (Float)"(InDouble=@n7)
  ]}
]
variables [
 armToStrap_L {type=number}
 armToStrap_R {type=number}
 armUp_L {type=number}
 armUp_R {type=number}
 legUp_L {type=number}
 legUp_R {type=number}
]
